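# Physics asset for the Ocean Rider medium female body's hips, covering a dynamic pack on the right side.
# Capsule bodies cover the pelvis, right thigh and two pack bones, with the pack chained from the pelvis through its base bone.
# Both pack joints lock twist and limit swing, the outer one to a single axis.
# The thigh-to-pelvis constraint leaves all angular motion free.
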PhysicsAsset /BRCosmetics/Characters/Player/Female/Medium/Bodies/F_MED_Ocean_Rider/Meshes/F_MED_Ocean_Rider_Dyn_Hips
bodies {
 dyn_pack_01_r [{shapeType=Capsule, length=8.1185, radius=3.8264, center=<6.2538 0 2.4145>, rotation=<rot 0 0 90>}]
 dyn_pack_base_r [{shapeType=Capsule, length=9.9623, radius=1.9826, center=<7.089 -0.0001 0>, rotation=<rot 0 0 90>}]
 pelvis [{shapeType=Capsule, length=7, radius=11.5625, center=<-4.0843 -0.7993 0>}]
 thigh_r [{shapeType=Capsule, length=27.9062, radius=8.2188, center=<23.3062 -0.0985 0.0134>, rotation=<rot 89.7505 166.3954 167.9212>}]
}
constraints [
 {
  bone1Name=dyn_pack_01_r
  bone2Name=dyn_pack_base_r
  swing1LimitDegrees=0
  swing1Motion=Locked
  swing2LimitDegrees=15
  twistLimitDegrees=0
  twistMotion=Locked
 }
 {
  bone1Name=thigh_r
  bone2Name=pelvis
  swing1LimitDegrees=45
  swing1Motion=Free
  swing2LimitDegrees=45
  swing2Motion=Free
  twistLimitDegrees=45
  twistMotion=Free
 }
 {bone1Name=dyn_pack_base_r, bone2Name=pelvis, swing1LimitDegrees=5, swing2LimitDegrees=30, twistLimitDegrees=0, twistMotion=Locked}
]
