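# Physics asset for the left-arm dynamic bracelet on the female medium Ocean Rider player body cosmetic.
# It holds two capsule bodies: a zero-length, effectively spherical one on dyn_bracelet_l and a small one on lowerarm_twist_01_l.
# One constraint joins the bracelet to the forearm twist bone with very tight swing and twist limit angles, allowing only slight rotational play.
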PhysicsAsset /BRCosmetics/Characters/Player/Female/Medium/Bodies/F_MED_Ocean_Rider/Meshes/F_MED_Ocean_Rider_Dyn_arm_l
bodies {
 dyn_bracelet_l [{shapeType=Capsule, length=0, radius=3.9829, center=<0.7589 -0.0294 -0.3941>, rotation=<rot 74.9385 83.7823 173.9947>}]
 lowerarm_twist_01_l [{shapeType=Capsule, length=0.5, radius=0.5, rotation=<rot 90 0 0>}]
}
constraints [{bone1Name=dyn_bracelet_l, bone2Name=lowerarm_twist_01_l, swing1LimitDegrees=2, swing2LimitDegrees=2, twistLimitDegrees=2}]
