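# Physics asset for the dynamic right-arm sleeve flap on the Ocean Rider medium female character body.
# Three capsule bodies form a short chain from the upper-arm twist bone through a master flap bone to the flap itself.
# The master's constraint to the arm has near-zero angular limits, while the flap's joint to the master locks twist and one swing axis, leaving the other swing axis for motion.
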PhysicsAsset /BRCosmetics/Characters/Player/Female/Medium/Bodies/F_MED_Ocean_Rider/Meshes/F_MED_Ocean_Rider_Dyn_arm_r
bodies {
 dyn_sleeve_flap_r [{shapeType=Capsule, length=1.5829, radius=0.5965, center=<0.4522 0 0>, rotation=<rot 90 0 0>}]
 dyn_sleeve_flap_r_MSTR [{shapeType=Capsule, length=0.5, radius=0.5}]
 upperarm_twist_02_r [{shapeType=Capsule, length=0.5, radius=0.5, rotation=<rot 90 0 0>}]
}
constraints [
 {
  bone1Name=dyn_sleeve_flap_r
  bone2Name=dyn_sleeve_flap_r_MSTR
  swing1LimitDegrees=5
  swing2LimitDegrees=45
  swing2Motion=Locked
  twistLimitDegrees=45
  twistMotion=Locked
 }
 {bone1Name=dyn_sleeve_flap_r_MSTR, bone2Name=upperarm_twist_02_r, swing1LimitDegrees=0.1, swing2LimitDegrees=0.1, twistLimitDegrees=0.1}
]
