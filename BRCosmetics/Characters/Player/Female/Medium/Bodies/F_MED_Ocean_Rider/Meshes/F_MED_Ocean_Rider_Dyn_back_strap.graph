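# Physics asset for the dynamic back strap on the female medium Ocean Rider player cosmetic body.
# A five-capsule dyn_tail chain hangs from a small dyn_tail_01_MSTR root body, which links to the pelvis with free angular motion.
# Swing limits widen progressively toward the tip, while twist stays tightly limited along the chain.
# Larger pelvis and thigh capsules provide body collision shapes, with both thighs angularly free relative to the pelvis.
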PhysicsAsset /BRCosmetics/Characters/Player/Female/Medium/Bodies/F_MED_Ocean_Rider/Meshes/F_MED_Ocean_Rider_Dyn_back_strap
bodies {
 dyn_tail_01 [{shapeType=Capsule, length=7.033, radius=1.2, center=<3.6637 0.1072 0.0011>, rotation=<rot 0 -91.3846 -89.6503>}]
 dyn_tail_01_MSTR [{shapeType=Capsule, length=0.5, radius=0.5}]
 dyn_tail_02 [{shapeType=Capsule, length=7.033, radius=1.2, center=<3.6637 0.1072 0.0011>, rotation=<rot 0 -91.3846 -89.6503>}]
 dyn_tail_03 [{shapeType=Capsule, length=9.0955, radius=1.2, center=<4.9706 0.0755 -0.0069>, rotation=<rot 0 -91.3846 -89.6503>}]
 dyn_tail_04 [{shapeType=Capsule, length=8.0017, radius=1.2, center=<4.175 0.0949 -0.002>, rotation=<rot 0 -91.3846 -89.6503>}]
 dyn_tail_05 [{shapeType=Capsule, length=7.5642, radius=1.2, center=<3.6637 0.1072 0.0011>, rotation=<rot 0 -91.3846 -89.6503>}]
 pelvis [{shapeType=Capsule, length=7, radius=11.5625, center=<-4.0843 -0.7993 0>}]
 thigh_l [{shapeType=Capsule, length=27.9062, radius=8.2188, center=<-22.9811 0.0987 -0.012>, rotation=<rot 89.7762 166.3999 167.9252>}]
 thigh_r [{shapeType=Capsule, length=27.9062, radius=8.2188, center=<23.3062 -0.0985 0.0134>, rotation=<rot 89.7505 166.3954 167.9212>}]
}
constraints [
 {bone1Name=dyn_tail_01, bone2Name=dyn_tail_01_MSTR, swing1LimitDegrees=5, swing2LimitDegrees=5, twistLimitDegrees=5}
 {bone1Name=dyn_tail_02, bone2Name=dyn_tail_01, swing1LimitDegrees=20, swing2LimitDegrees=20, twistLimitDegrees=5}
 {bone1Name=dyn_tail_03, bone2Name=dyn_tail_02, swing1LimitDegrees=25, swing2LimitDegrees=25, twistLimitDegrees=5}
 {bone1Name=dyn_tail_04, bone2Name=dyn_tail_03, swing1LimitDegrees=30, swing2LimitDegrees=30, twistLimitDegrees=5}
 {bone1Name=dyn_tail_05, bone2Name=dyn_tail_04, swing1LimitDegrees=35, swing2LimitDegrees=35, twistLimitDegrees=5}
 {
  bone1Name=thigh_l
  bone2Name=pelvis
  swing1LimitDegrees=45
  swing1Motion=Free
  swing2LimitDegrees=45
  swing2Motion=Free
  twistLimitDegrees=45
  twistMotion=Free
 }
 {
  bone1Name=thigh_r
  bone2Name=pelvis
  swing1LimitDegrees=45
  swing1Motion=Free
  swing2LimitDegrees=45
  swing2Motion=Free
  twistLimitDegrees=45
  twistMotion=Free
 }
 {
  bone1Name=dyn_tail_01_MSTR
  bone2Name=pelvis
  swing1LimitDegrees=45
  swing1Motion=Free
  swing2LimitDegrees=45
  swing2Motion=Free
  twistLimitDegrees=45
  twistMotion=Free
 }
]
