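# Physics asset for the left leg of the F_MED_Ocean_Rider player character, setting up a four-bone dyn_sleeve_l dangling chain anchored at thigh_l.
# Capsule bodies cover each chain bone, with a tiny capsule on the MSTR root bone and a large one on the thigh.
# Swing and twist limits are tight near the root and loosen toward the tip, while the root-to-thigh joint leaves all angular motion free.
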PhysicsAsset /BRCosmetics/Characters/Player/Female/Medium/Bodies/F_MED_Ocean_Rider/Meshes/F_MED_Ocean_Rider_Dyn_leg_l
bodies {
 dyn_sleeve_l_01 [{shapeType=Capsule, length=15, radius=1.15, shapeName=dyn_sleeve_l_01_capsule, center=<6.1 0 0>, rotation=<rot 0 0 90>}]
 dyn_sleeve_l_01_MSTR [{shapeType=Capsule, length=0.5, radius=0.5}]
 dyn_sleeve_l_02 [{shapeType=Capsule, length=15, radius=1.15, shapeName=dyn_sleeve_l_02_capsule, center=<6.1 0 0>, rotation=<rot 0 0 90>}]
 dyn_sleeve_l_03 [{shapeType=Capsule, length=15, radius=1.15, shapeName=dyn_sleeve_l_03_capsule, center=<6.1 0 0>, rotation=<rot 0 0 90>}]
 dyn_sleeve_l_04 [{shapeType=Capsule, length=15, radius=1.15, shapeName=dyn_sleeve_l_04_capsule, center=<3.0638 0 0>, rotation=<rot 0 0 90>}]
 thigh_l [{shapeType=Capsule, length=27.9062, radius=8.2188, center=<-22.9811 0.0987 -0.012>, rotation=<rot 89.7762 166.3999 167.9252>}]
}
constraints [
 {bone1Name=dyn_sleeve_l_01, bone2Name=dyn_sleeve_l_01_MSTR, swing1LimitDegrees=5, swing2LimitDegrees=5, twistLimitDegrees=15}
 {bone1Name=dyn_sleeve_l_02, bone2Name=dyn_sleeve_l_01, swing1LimitDegrees=15, swing2LimitDegrees=15, twistLimitDegrees=15}
 {bone1Name=dyn_sleeve_l_03, bone2Name=dyn_sleeve_l_02, swing1LimitDegrees=30, swing2LimitDegrees=30, twistLimitDegrees=45}
 {bone1Name=dyn_sleeve_l_04, bone2Name=dyn_sleeve_l_03, swing1LimitDegrees=45, swing2LimitDegrees=45, twistLimitDegrees=45}
 {
  bone1Name=dyn_sleeve_l_01_MSTR
  bone2Name=thigh_l
  swing1LimitDegrees=45
  swing1Motion=Free
  swing2LimitDegrees=45
  swing2Motion=Free
  twistLimitDegrees=45
  twistMotion=Free
 }
]
