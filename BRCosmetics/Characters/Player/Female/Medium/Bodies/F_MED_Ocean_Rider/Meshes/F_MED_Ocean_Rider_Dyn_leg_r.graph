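# Physics asset for the right-leg dynamic chain of the female medium Ocean Rider cosmetic body, named as a sleeve.
# Four slim capsule bodies form the dyn_sleeve_r chain, hung from a small MSTR root body that is constrained to a large thigh_r capsule.
# The root-to-thigh joint has free swing and twist, while swing and twist limit angles along the chain start tight and widen toward the tip.
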PhysicsAsset /BRCosmetics/Characters/Player/Female/Medium/Bodies/F_MED_Ocean_Rider/Meshes/F_MED_Ocean_Rider_Dyn_leg_r
bodies {
 dyn_sleeve_r_01 [{shapeType=Capsule, length=15, radius=1.15, shapeName=dyn_sleeve_r_01_capsule, center=<6.1 0.0001 0>, rotation=<rot 0 0 90>}]
 dyn_sleeve_r_01_MSTR [{shapeType=Capsule, length=0.5, radius=0.5}]
 dyn_sleeve_r_02 [{shapeType=Capsule, length=15, radius=1.15, shapeName=dyn_sleeve_r_02_capsule, center=<6.1 0.0001 0>, rotation=<rot 0 0 90>}]
 dyn_sleeve_r_03 [{shapeType=Capsule, length=15, radius=1.15, shapeName=dyn_sleeve_r_03_capsule, center=<6.1 0.0001 0>, rotation=<rot 0 0 90>}]
 dyn_sleeve_r_04 [{shapeType=Capsule, length=15, radius=1.15, shapeName=dyn_sleeve_r_04_capsule, center=<3.0638 0.0001 0>, rotation=<rot 0 0 90>}]
 thigh_r [{shapeType=Capsule, length=27.9062, radius=8.2188, center=<23.3062 -0.0985 0.0134>, rotation=<rot 89.7505 166.3954 167.9212>}]
}
constraints [
 {bone1Name=dyn_sleeve_r_01, bone2Name=dyn_sleeve_r_01_MSTR, swing1LimitDegrees=5, swing2LimitDegrees=5, twistLimitDegrees=15}
 {bone1Name=dyn_sleeve_r_02, bone2Name=dyn_sleeve_r_01, swing1LimitDegrees=5, swing2LimitDegrees=5, twistLimitDegrees=15}
 {bone1Name=dyn_sleeve_r_03, bone2Name=dyn_sleeve_r_02, swing1LimitDegrees=30, swing2LimitDegrees=30, twistLimitDegrees=45}
 {bone1Name=dyn_sleeve_r_04, bone2Name=dyn_sleeve_r_03, swing1LimitDegrees=45, swing2LimitDegrees=45, twistLimitDegrees=45}
 {
  bone1Name=dyn_sleeve_r_01_MSTR
  bone2Name=thigh_r
  swing1LimitDegrees=45
  swing1Motion=Free
  swing2LimitDegrees=45
  swing2Motion=Free
  twistLimitDegrees=45
  twistMotion=Free
 }
]
