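# Physics asset for the right backpack zipper pull on the Ocean Rider medium female player cosmetic body.
# Small capsules on the zipper bone and its MSTR bone are joined by the only constraint, with narrow swing limits and a zero twist limit.
# A larger capsule on the right pack bone has no constraint, giving the pack a collision volume alongside the zipper.
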PhysicsAsset /BRCosmetics/Characters/Player/Female/Medium/Bodies/F_MED_Ocean_Rider/Meshes/F_MED_Ocean_Rider_Dyn_pack_zipper
bodies {
 dyn_pack_01_r [{
   shapeType=Capsule
   length=13.7674
   radius=8.7236
   shapeName=dyn_pack_01_r_capsule
   center=<7.831 -0.0016 0.7586>
   rotation=<rot 0 -1.3287 90>
  }]
 dyn_pack_zipper_r [{shapeType=Capsule, length=2.662, radius=0.6949, center=<1.5 0 0>, rotation=<rot -90 0 0>}]
 dyn_pack_zipper_r_MSTR [{shapeType=Capsule, length=0.5, radius=0.5}]
}
constraints [{bone1Name=dyn_pack_zipper_r, bone2Name=dyn_pack_zipper_r_MSTR, swing1LimitDegrees=10, swing2LimitDegrees=15, twistLimitDegrees=0}]
