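# Physics asset for the Ocean Rider medium female player body that drives secondary motion on the jacket and necklace.
# Eight two-box jacket chains hang from small MSTR capsules joined to spine_05, with limited swing, tight twist and one swing axis locked at the side roots.
# Free-rotation constraints tie the lower center back panel to its neighbors.
# A three-link necklace hangs from spine_05, and spine capsules form the torso volume.
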PhysicsAsset /Game/Characters/Player/Female/Medium/Bodies/F_MED_Ocean_Rider/Meshes/F_MED_Ocean_Rider_Dyn
bodies {
 dyn_jacket_bk_01 [{shapeType=Box, extentX=6.875, extentY=9.9375, extentZ=1.8016, center=<3.0638 0.0001 0>}]
 dyn_jacket_bk_01_MSTR [{shapeType=Capsule, length=0.5, radius=0.5}]
 dyn_jacket_bk_02 [{shapeType=Box, extentX=6.875, extentY=9.9375, extentZ=0.9375, center=<3.0638 0.0001 0>}]
 dyn_jacket_bk_l_01 [{shapeType=Box, extentX=6.875, extentY=14.7355, extentZ=1.6875, center=<3.0638 0.0001 0>}]
 dyn_jacket_bk_l_01_MSTR [{shapeType=Capsule, length=0.5, radius=0.5}]
 dyn_jacket_bk_l_02 [{shapeType=Box, extentX=6.875, extentY=9.9375, extentZ=1.878, center=<2.9803 0 -0.4733>, rotation=<rot -10 0 0>}]
 dyn_jacket_bk_r_01 [{shapeType=Box, extentX=6.875, extentY=14.7355, extentZ=1.6875, center=<3.0638 0.0001 0>}]
 dyn_jacket_bk_r_01_MSTR [{shapeType=Capsule, length=0.5, radius=0.5}]
 dyn_jacket_bk_r_02 [{shapeType=Box, extentX=6.875, extentY=9.9375, extentZ=1.878, center=<2.9803 0 -0.4733>, rotation=<rot -10 0 0>}]
 dyn_jacket_fr_01 [{shapeType=Box, extentX=6.875, extentY=9.9375, extentZ=0.9375, center=<3.0638 0.0001 0>}]
 dyn_jacket_fr_01_MSTR [{shapeType=Capsule, length=0.5, radius=0.5}]
 dyn_jacket_fr_02 [{shapeType=Box, extentX=6.875, extentY=9.9375, extentZ=0.9375, center=<3.0638 0.0001 0>}]
 dyn_jacket_fr_l_01 [{shapeType=Box, extentX=6.875, extentY=9.9375, extentZ=0.9375, center=<3.0638 0.0001 0>}]
 dyn_jacket_fr_l_01_MSTR [{shapeType=Capsule, length=0.5, radius=0.5}]
 dyn_jacket_fr_l_02 [{shapeType=Box, extentX=6.875, extentY=9.9375, extentZ=0.9375, center=<3.0638 0.0001 0>}]
 dyn_jacket_fr_r_01 [{shapeType=Box, extentX=6.875, extentY=9.9375, extentZ=0.9375, center=<3.0638 0.0001 0>}]
 dyn_jacket_fr_r_01_MSTR [{shapeType=Capsule, length=0.5, radius=0.5}]
 dyn_jacket_fr_r_02 [{shapeType=Box, extentX=6.875, extentY=9.9375, extentZ=0.9375, center=<3.0638 0.0001 0>}]
 dyn_jacket_l_01 [{shapeType=Box, extentX=6.875, extentY=14.7355, extentZ=0.9375, center=<3.0638 0.0001 0>}]
 dyn_jacket_l_01_MSTR [{shapeType=Capsule, length=0.5, radius=0.5}]
 dyn_jacket_l_02 [{shapeType=Box, extentX=6.875, extentY=14.7355, extentZ=0.9375, center=<3.0638 0.0001 0>}]
 dyn_jacket_r_01 [{shapeType=Box, extentX=6.875, extentY=14.7355, extentZ=0.9375, center=<3.0638 0.0001 0>}]
 dyn_jacket_r_01_MSTR [{shapeType=Capsule, length=0.5, radius=0.5}]
 dyn_jacket_r_02 [{shapeType=Box, extentX=6.875, extentY=14.7355, extentZ=0.9375, center=<3.0638 0.0001 0>}]
 dyn_necklace_01 [{shapeType=Capsule, length=5.4702, radius=1.0961, shapeName=dyn_necklace_01_capsule, center=<1.5474 0 -0.0257>, rotation=<rot -87.9401 0 0>}]
 dyn_necklace_02 [{
   shapeType=Capsule
   length=6.2046
   radius=1.2082
   shapeName=dyn_necklace_02_capsule
   center=<1.4221 0 0.0107>
   rotation=<rot -89.5698 -0.0002 0.0002>
  }]
 dyn_necklace_03 [{shapeType=Sphere, radius=2.0938, center=<2.0878 0 0.6184>}]
 spine_02 [{shapeType=Capsule, length=1.4375, radius=9.0312}]
 spine_03 [{shapeType=Capsule, length=1.6875, radius=10.4062, center=<1.96 -0.457 0>}]
 spine_04 [
  {shapeType=Capsule, length=5.0625, radius=8.625, center=<-0.9609 -0.1808 0.0001>}
  {shapeType=Capsule, length=5.1596, radius=9.0312, center=<-3.8702 3.5926 0>}
  {shapeType=Capsule, length=7.2846, radius=6.9062, center=<4.2935 4.9263 0>}
  {shapeType=Capsule, length=5.7812, radius=8.625, center=<2.7483 0.0648 0.0001>}
 ]
 spine_05 [
  {shapeType=Capsule, length=4.5312, radius=9.0312, center=<-1.2073 0.8192 0>}
  {shapeType=Capsule, length=5.3684, radius=8.0625, center=<-6.1062 5.8463 0>}
  {shapeType=Capsule, length=4.5312, radius=9.0312, center=<-4.6459 0.8192 0>}
 ]
}
constraints [
 {bone1Name=dyn_jacket_l_02, bone2Name=dyn_jacket_l_01, swing1LimitDegrees=8, swing2LimitDegrees=45, twistLimitDegrees=1}
 {bone1Name=dyn_jacket_bk_l_02, bone2Name=dyn_jacket_bk_l_01, swing1LimitDegrees=30, swing2LimitDegrees=45, twistLimitDegrees=1}
 {bone1Name=dyn_jacket_bk_02, bone2Name=dyn_jacket_bk_01, swing1LimitDegrees=30, swing2LimitDegrees=45, twistLimitDegrees=1}
 {bone1Name=dyn_jacket_fr_l_02, bone2Name=dyn_jacket_fr_l_01, swing1LimitDegrees=30, swing2LimitDegrees=45, twistLimitDegrees=1}
 {bone1Name=dyn_jacket_fr_02, bone2Name=dyn_jacket_fr_01, swing1LimitDegrees=30, swing2LimitDegrees=45, twistLimitDegrees=1}
 {bone1Name=dyn_necklace_02, bone2Name=dyn_necklace_01, swing1LimitDegrees=30, swing2LimitDegrees=30, twistLimitDegrees=5}
 {bone1Name=dyn_necklace_03, bone2Name=dyn_necklace_02, swing1LimitDegrees=30, swing2LimitDegrees=30, twistLimitDegrees=5}
 {bone1Name=dyn_jacket_fr_r_02, bone2Name=dyn_jacket_fr_r_01, swing1LimitDegrees=30, swing2LimitDegrees=45, twistLimitDegrees=1}
 {bone1Name=dyn_jacket_bk_r_02, bone2Name=dyn_jacket_bk_r_01, swing1LimitDegrees=30, swing2LimitDegrees=45, twistLimitDegrees=1}
 {bone1Name=dyn_jacket_r_02, bone2Name=dyn_jacket_r_01, swing1LimitDegrees=8, swing2LimitDegrees=45, twistLimitDegrees=1}
 {
  bone1Name=dyn_jacket_l_01
  bone2Name=dyn_jacket_l_01_MSTR
  swing1LimitDegrees=2
  swing1Motion=Locked
  swing2LimitDegrees=20
  twistLimitDegrees=1
 }
 {bone1Name=dyn_jacket_bk_l_01, bone2Name=dyn_jacket_bk_l_01_MSTR, swing1LimitDegrees=8, swing2LimitDegrees=45, twistLimitDegrees=1}
 {bone1Name=dyn_jacket_bk_01, bone2Name=dyn_jacket_bk_01_MSTR, swing1LimitDegrees=8, swing2LimitDegrees=45, twistLimitDegrees=1}
 {bone1Name=dyn_jacket_fr_l_01, bone2Name=dyn_jacket_fr_l_01_MSTR, swing1LimitDegrees=5, swing2LimitDegrees=45, twistLimitDegrees=1}
 {bone1Name=dyn_jacket_fr_01, bone2Name=dyn_jacket_fr_01_MSTR, swing1LimitDegrees=5, swing2LimitDegrees=45, twistLimitDegrees=1}
 {bone1Name=dyn_necklace_01, bone2Name=spine_05, swing1LimitDegrees=30, swing2LimitDegrees=30, twistLimitDegrees=5}
 {bone1Name=dyn_jacket_fr_r_01, bone2Name=dyn_jacket_fr_r_01_MSTR, swing1LimitDegrees=5, swing2LimitDegrees=45, twistLimitDegrees=1}
 {bone1Name=dyn_jacket_bk_r_01, bone2Name=dyn_jacket_bk_r_01_MSTR, swing1LimitDegrees=8, swing2LimitDegrees=45, twistLimitDegrees=1}
 {
  bone1Name=dyn_jacket_r_01
  bone2Name=dyn_jacket_r_01_MSTR
  swing1LimitDegrees=2
  swing1Motion=Locked
  swing2LimitDegrees=20
  twistLimitDegrees=1
 }
 {
  bone1Name=spine_05
  bone2Name=spine_04
  swing1LimitDegrees=45
  swing1Motion=Free
  swing2LimitDegrees=45
  swing2Motion=Free
  twistLimitDegrees=45
  twistMotion=Free
 }
 {
  bone1Name=spine_04
  bone2Name=spine_03
  swing1LimitDegrees=45
  swing1Motion=Free
  swing2LimitDegrees=45
  swing2Motion=Free
  twistLimitDegrees=45
  twistMotion=Free
 }
 {
  bone1Name=spine_03
  bone2Name=spine_02
  swing1LimitDegrees=45
  swing1Motion=Free
  swing2LimitDegrees=45
  swing2Motion=Free
  twistLimitDegrees=45
  twistMotion=Free
 }
 {
  bone1Name=dyn_jacket_l_01_MSTR
  bone2Name=spine_05
  swing1LimitDegrees=45
  swing1Motion=Free
  swing2LimitDegrees=45
  swing2Motion=Free
  twistLimitDegrees=45
  twistMotion=Free
 }
 {
  bone1Name=dyn_jacket_bk_l_01_MSTR
  bone2Name=spine_05
  swing1LimitDegrees=45
  swing1Motion=Free
  swing2LimitDegrees=45
  swing2Motion=Free
  twistLimitDegrees=45
  twistMotion=Free
 }
 {
  bone1Name=dyn_jacket_bk_01_MSTR
  bone2Name=spine_05
  swing1LimitDegrees=45
  swing1Motion=Free
  swing2LimitDegrees=45
  swing2Motion=Free
  twistLimitDegrees=45
  twistMotion=Free
 }
 {
  bone1Name=dyn_jacket_fr_l_01_MSTR
  bone2Name=spine_05
  swing1LimitDegrees=45
  swing1Motion=Free
  swing2LimitDegrees=45
  swing2Motion=Free
  twistLimitDegrees=45
  twistMotion=Free
 }
 {
  bone1Name=dyn_jacket_fr_01_MSTR
  bone2Name=spine_05
  swing1LimitDegrees=45
  swing1Motion=Free
  swing2LimitDegrees=45
  swing2Motion=Free
  twistLimitDegrees=45
  twistMotion=Free
 }
 {
  bone1Name=dyn_jacket_fr_r_01_MSTR
  bone2Name=spine_05
  swing1LimitDegrees=45
  swing1Motion=Free
  swing2LimitDegrees=45
  swing2Motion=Free
  twistLimitDegrees=45
  twistMotion=Free
 }
 {
  bone1Name=dyn_jacket_bk_r_01_MSTR
  bone2Name=spine_05
  swing1LimitDegrees=45
  swing1Motion=Free
  swing2LimitDegrees=45
  swing2Motion=Free
  twistLimitDegrees=45
  twistMotion=Free
 }
 {
  bone1Name=dyn_jacket_r_01_MSTR
  bone2Name=spine_05
  swing1LimitDegrees=45
  swing1Motion=Free
  swing2LimitDegrees=45
  swing2Motion=Free
  twistLimitDegrees=45
  twistMotion=Free
 }
 {
  bone1Name=dyn_jacket_bk_02
  bone2Name=dyn_jacket_bk_l_02
  swing1LimitDegrees=45
  swing1Motion=Free
  swing2LimitDegrees=45
  swing2Motion=Free
  twistLimitDegrees=45
  twistMotion=Free
 }
 {
  bone1Name=dyn_jacket_bk_02
  bone2Name=dyn_jacket_bk_r_02
  swing1LimitDegrees=45
  swing1Motion=Free
  swing2LimitDegrees=45
  swing2Motion=Free
  twistLimitDegrees=45
  twistMotion=Free
 }
]
