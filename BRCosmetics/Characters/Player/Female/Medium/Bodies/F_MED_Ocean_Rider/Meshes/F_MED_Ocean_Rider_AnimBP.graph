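# Post-process-style cosmetic AnimBP for the female Ocean Rider body, copying the parent mesh pose and adding secondary motion to jacket, necklace, straps and pack.
# Curve-driven morph and collar corrections, weighted bone constraints and additive offsets feed jacket AnimDynamics chains, then hip, strap and arm RigidBody simulations.
# That cached result branches into standard and crouch leg simulations, blended by bool before pack simulation.
# The EventGraph computes state-based gravity vectors and a vest flag.
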AnimBlueprint /BRCosmetics/Characters/Player/Female/Medium/Bodies/F_MED_Ocean_Rider/Meshes/F_MED_Ocean_Rider_AnimBP
target_skeleton=&/F_MED_Ocean_Rider_Skeleton
graphs [
 AnimGraph {nodes [
   RigidBody() {
    bClampLinearTranslationLimitToRefPose=true
    bOverrideWorldGravity=true
    overridePhysicsAsset=/Game/Characters/Player/Female/Medium/Bodies/F_MED_Ocean_Rider/Meshes/F_MED_Ocean_Rider_Dyn
    simulationSpace=BaseBoneSpace
    alphaBoolBlend {blendOption=Linear}
    baseBoneRef {boneName=spine_01}
    componentAppliedLinearAccClamp=<3000 3000 3000>
    componentLinearAccScale=<0.3 0.25 0.25>
    componentLinearVelScale=<0.3 0.25 0.25>
    overrideWorldGravity {z=-980}
   }
   n1 = RigidBody!(ComponentPose=Local To Component(LocalPose=Use cached pose 'crouchCache'()), Alpha, OverrideWorldGravity:Vector="0,0,-980", ExternalForce) -> (Pose) {
    bClampLinearTranslationLimitToRefPose=true
    bOverrideWorldGravity=true
    overridePhysicsAsset=&/F_MED_Ocean_Rider_Dyn_leg_l_Crouch
    simulationSpace=BaseBoneSpace
    alphaBoolBlend {blendOption=Linear}
    baseBoneRef {boneName=thigh_l}
    componentAppliedLinearAccClamp=<3000 3000 3000>
    componentLinearAccScale=<0.4 0.35 0.35>
    componentLinearVelScale=<0.4 0.35 0.35>
    overrideWorldGravity {z=-980}
   }
   n2 = RigidBody(ComponentPose=@n1, OverrideWorldGravity="0,0,-980") {
    bClampLinearTranslationLimitToRefPose=true
    bOverrideWorldGravity=true
    overridePhysicsAsset=&/F_MED_Ocean_Rider_Dyn_leg_r_Crouch
    simulationSpace=BaseBoneSpace
    alphaBoolBlend {blendOption=Linear}
    baseBoneRef {boneName=thigh_r}
    componentAppliedLinearAccClamp=<3000 3000 3000>
    componentLinearAccScale=<0.4 0.35 0.35>
    componentLinearVelScale=<0.4 0.35 0.35>
    overrideWorldGravity {z=-980}
   }
   n3 = Component To Local(ComponentPose=@n2)
   n4 = RigidBody(ComponentPose=Local To Component(LocalPose=Use cached pose 'crouchCache'()), OverrideWorldGravity="0,0,-980") {
    bClampLinearTranslationLimitToRefPose=true
    bOverrideWorldGravity=true
    overridePhysicsAsset=&/F_MED_Ocean_Rider_Dyn_leg_l
    simulationSpace=BaseBoneSpace
    alphaBoolBlend {blendOption=Linear}
    baseBoneRef {boneName=thigh_l}
    componentAppliedLinearAccClamp=<3000 3000 3000>
    componentLinearAccScale=<0.4 0.35 0.35>
    componentLinearVelScale=<0.4 0.35 0.35>
    overrideWorldGravity {z=-980}
   }
   n5 = RigidBody(ComponentPose=@n4, OverrideWorldGravity="0,0,-980") {
    bClampLinearTranslationLimitToRefPose=true
    bOverrideWorldGravity=true
    overridePhysicsAsset=&/F_MED_Ocean_Rider_Dyn_leg_r
    simulationSpace=BaseBoneSpace
    alphaBoolBlend {blendOption=Linear}
    baseBoneRef {boneName=thigh_r}
    componentAppliedLinearAccClamp=<3000 3000 3000>
    componentLinearAccScale=<0.4 0.35 0.35>
    componentLinearVelScale=<0.4 0.35 0.35>
    overrideWorldGravity {z=-980}
   }
   n6 = Component To Local(ComponentPose=@n5)
   n7 = "Constraint - dyn_pack_grenade_r_MSTR (PPPPPPP)"!(
    ComponentPose=Local To Component(LocalPose=Blend Poses by bool(BlendPose_0=@n3, BlendPose_1=@n6) {blendType=Linear})
    Alpha=1
    ConstraintWeights_0=0.3743
    ConstraintWeights_1:float=0.0774
    ConstraintWeights_2:float=0.3881
    ConstraintWeights_3:float=0
    ConstraintWeights_4:float=0
    ConstraintWeights_5:float=0.1602
    ConstraintWeights_6:float=0
   ) -> (Pose) {
    alphaBoolBlend {blendOption=Linear}
    boneToModify {boneName=dyn_pack_grenade_r_MSTR}
    constraintSetup [
     {targetBone {boneName=dyn_pack_01_r}}
     {targetBone {boneName=thigh_r}}
     {targetBone {boneName=thigh_twist_01_r}}
     {targetBone {boneName=spine_01}}
     {targetBone {boneName=spine_02}}
     {targetBone {boneName=dyn_pack_base_r}}
     {targetBone {boneName=pelvis}}
    ]
    constraintWeights [1,1,1,1,1,1,1]
   }
   n8 = "Constraint - dyn_pack_zipper_r_MSTR (PPPPPPP)"!(
    ComponentPose=@n7
    Alpha=1
    ConstraintWeights_0=0.0041
    ConstraintWeights_1:float=0.0005
    ConstraintWeights_2:float=0.0985
    ConstraintWeights_3:float=0
    ConstraintWeights_4:float=0
    ConstraintWeights_5:float=0.8957
    ConstraintWeights_6:float=0.0012
   ) -> (Pose) {
    alphaBoolBlend {blendOption=Linear}
    boneToModify {boneName=dyn_pack_zipper_r_MSTR}
    constraintSetup [
     {targetBone {boneName=dyn_pack_01_r}}
     {targetBone {boneName=thigh_r}}
     {targetBone {boneName=thigh_twist_01_r}}
     {targetBone {boneName=spine_01}}
     {targetBone {boneName=spine_02}}
     {targetBone {boneName=dyn_pack_base_r}}
     {targetBone {boneName=pelvis}}
    ]
    constraintWeights [1,1,1,1,1,1,1]
   }
   n9 = RigidBody!(ComponentPose=@n8, Alpha, ExternalForce) -> (Pose) {
    bClampLinearTranslationLimitToRefPose=true
    bOverrideWorldGravity=true
    overridePhysicsAsset=&/F_MED_Ocean_Rider_Dyn_pack_zipper
    simulationSpace=BaseBoneSpace
    alphaBoolBlend {blendOption=Linear}
    baseBoneRef {boneName=pelvis}
    componentAppliedLinearAccClamp=<2000 2000 2000>
    componentLinearAccScale=<0.4 0.3 0.3>
    componentLinearVelScale=<0.4 0.3 0.3>
    overrideWorldGravity {z=-980}
   }
   Output Pose(Result=Component To Local(ComponentPose=@n9)) {layerGroup=None}
   n10 = Copy Pose From Mesh() {bUseAttachedParent=true}
   n11 = Get LeftArmDown() -> (LeftArmDown:float)
   n12 = "Transform (Modify) Bone - Bone: dyn_collar_bk"(
    ComponentPose=Local To Component(LocalPose=Modify Curve!(SourcePose=@n10, CurveValues_0=1, CurveValues_1:float=1, CurveValues_2:float=1, Alpha=@n11) -> (Pose) {curveNames [Body_T12_MORPH, Jacket_T2_MORPH, Shirt_T12_MORPH], curveValues [0,0,0]})
    Alpha=Get HeadRot() -> (HeadRot:float)
    Translation="-6.541,0.000,-1.452"
    Rotation="0,0,0"
    Scale="1,1,1"
   ) {
    rotationMode=BMM_Additive
    rotationSpace=BCS_BoneSpace
    translationMode=BMM_Additive
    translationSpace=BCS_BoneSpace
    alphaBoolBlend {blendOption=Linear}
    boneToModify {boneName=dyn_collar_bk}
    translation {x=-6.541, z=-1.452}
   }
   n13 = "Constraint - dyn_tail_01_MSTR (PP)"!(ComponentPose=@n12, Alpha=1, ConstraintWeights_0=0.9584, ConstraintWeights_1:float=0.0416) -> (Pose) {
    alphaBoolBlend {blendOption=Linear}
    boneToModify {boneName=dyn_tail_01_MSTR}
    constraintSetup [{targetBone {boneName=pelvis}}, {targetBone {boneName=thigh_twist_01_r}}]
    constraintWeights [1,1]
   }
   n14 = "Constraint - dyn_jacket_l_01_MSTR (PPPP)"!(
    ComponentPose=@n13
    Alpha=1
    ConstraintWeights_0=0.1832
    ConstraintWeights_1:float=0.4059
    ConstraintWeights_2:float=0.1791
    ConstraintWeights_3:float=0.2318
   ) -> (Pose) {
    alphaBoolBlend {blendOption=Linear}
    boneToModify {boneName=dyn_jacket_l_01_MSTR}
    constraintSetup [
     {targetBone {boneName=clavicle_l}}
     {targetBone {boneName=spine_04}}
     {targetBone {boneName=spine_05}}
     {targetBone {boneName=upperarm_twist_01_l}}
    ]
    constraintWeights [1,1,1,1]
   }
   n15 = "Constraint - dyn_jacket_r_01_MSTR (PPPP)"!(
    ComponentPose=@n14
    Alpha=1
    ConstraintWeights_0=0.2318
    ConstraintWeights_1:float=0.1832
    ConstraintWeights_2:float=0.4059
    ConstraintWeights_3:float=0.1791
   ) -> (Pose) {
    alphaBoolBlend {blendOption=Linear}
    boneToModify {boneName=dyn_jacket_r_01_MSTR}
    constraintSetup [
     {targetBone {boneName=upperarm_twist_01_r}}
     {targetBone {boneName=clavicle_r}}
     {targetBone {boneName=spine_04}}
     {targetBone {boneName=spine_05}}
    ]
    constraintWeights [1,1,1,1]
   }
   n16 = "Constraint - dyn_jacket_bk_l_01_MSTR (PPPPPPP)"!(
    ComponentPose=@n15
    Alpha=1
    ConstraintWeights_0=0.436
    ConstraintWeights_1:float=0.0017
    ConstraintWeights_2:float=0.0076
    ConstraintWeights_3:float=0.2473
    ConstraintWeights_4:float=0.2693
    ConstraintWeights_5:float=0.0336
    ConstraintWeights_6:float=0.0045
   ) -> (Pose) {
    alphaBoolBlend {blendOption=Linear}
    boneToModify {boneName=dyn_jacket_bk_l_01_MSTR}
    constraintSetup [
     {targetBone {boneName=clavicle_l}}
     {targetBone {boneName=spine_02}}
     {targetBone {boneName=spine_03}}
     {targetBone {boneName=spine_04}}
     {targetBone {boneName=spine_05}}
     {targetBone {boneName=upperarm_twist_01_l}}
     {targetBone {boneName=clavicle_r}}
    ]
    constraintWeights [1,1,1,1,1,1,1]
   }
   n17 = "Constraint - dyn_jacket_bk_01_MSTR (PPPP)"!(
    ComponentPose=@n16
    Alpha=1
    ConstraintWeights_0=0.1572
    ConstraintWeights_1:float=0.1707
    ConstraintWeights_2:float=0.5149
    ConstraintWeights_3:float=0.1572
   ) -> (Pose) {
    alphaBoolBlend {blendOption=Linear}
    boneToModify {boneName=dyn_jacket_bk_01_MSTR}
    constraintSetup [{targetBone {boneName=clavicle_r}}, {targetBone {boneName=spine_04}}, {targetBone {boneName=spine_05}}, {targetBone {boneName=clavicle_l}}]
    constraintWeights [1,1,1,1]
   }
   n18 = "Constraint - dyn_jacket_fr_l_01_MSTR (PP)"!(ComponentPose=@n17, Alpha=1, ConstraintWeights_0=0.92, ConstraintWeights_1:float=0.08) -> (Pose) {
    alphaBoolBlend {blendOption=Linear}
    boneToModify {boneName=dyn_jacket_fr_l_01_MSTR}
    constraintSetup [{targetBone {boneName=spine_04}}, {targetBone {boneName=spine_05}}]
    constraintWeights [1,1]
   }
   n19 = "Constraint - dyn_jacket_fr_01_MSTR (PP)"!(ComponentPose=@n18, Alpha=1, ConstraintWeights_0=0.9324, ConstraintWeights_1:float=0.0676) -> (Pose) {
    alphaBoolBlend {blendOption=Linear}
    boneToModify {boneName=dyn_jacket_fr_01_MSTR}
    constraintSetup [{targetBone {boneName=spine_04}}, {targetBone {boneName=spine_05}}]
    constraintWeights [1,1]
   }
   n20 = "Constraint - dyn_jacket_fr_r_01_MSTR (PP)"!(ComponentPose=@n19, Alpha=1, ConstraintWeights_0=0.92, ConstraintWeights_1:float=0.08) -> (Pose) {
    alphaBoolBlend {blendOption=Linear}
    boneToModify {boneName=dyn_jacket_fr_r_01_MSTR}
    constraintSetup [{targetBone {boneName=spine_04}}, {targetBone {boneName=spine_05}}]
    constraintWeights [1,1]
   }
   n21 = "Constraint - dyn_jacket_bk_r_01_MSTR (PPPP)"!(
    ComponentPose=@n20
    Alpha=1
    ConstraintWeights_0=0.0856
    ConstraintWeights_1:float=0.4369
    ConstraintWeights_2:float=0.2297
    ConstraintWeights_3:float=0.2478
   ) -> (Pose) {
    alphaBoolBlend {blendOption=Linear}
    boneToModify {boneName=dyn_jacket_bk_r_01_MSTR}
    constraintSetup [
     {targetBone {boneName=upperarm_twist_01_r}}
     {targetBone {boneName=clavicle_r}}
     {targetBone {boneName=spine_04}}
     {targetBone {boneName=spine_05}}
    ]
    constraintWeights [1,1,1,1]
   }
   n22 = "Constraint - dyn_jacket_l_02 (P)"(ComponentPose=@n21, Alpha=1, ConstraintWeights_0=1) {
    alphaBoolBlend {blendOption=Linear}
    boneToModify {boneName=dyn_jacket_l_02}
    constraintSetup [{targetBone {boneName=spine_04}}]
    constraintWeights [1]
   }
   n23 = "Constraint - dyn_jacket_bk_l_02 (P)"(ComponentPose=@n22, Alpha=1, ConstraintWeights_0=1) {
    alphaBoolBlend {blendOption=Linear}
    boneToModify {boneName=dyn_jacket_bk_l_02}
    constraintSetup [{targetBone {boneName=spine_04}}]
    constraintWeights [1]
   }
   n24 = "Constraint - dyn_jacket_bk_02 (P)"(ComponentPose=@n23, Alpha=1, ConstraintWeights_0=1) {
    alphaBoolBlend {blendOption=Linear}
    boneToModify {boneName=dyn_jacket_bk_02}
    constraintSetup [{targetBone {boneName=spine_04}}]
    constraintWeights [1]
   }
   n25 = "Constraint - dyn_jacket_bk_r_02 (P)"(ComponentPose=@n24, Alpha=1, ConstraintWeights_0=1) {
    alphaBoolBlend {blendOption=Linear}
    boneToModify {boneName=dyn_jacket_bk_r_02}
    constraintSetup [{targetBone {boneName=spine_04}}]
    constraintWeights [1]
   }
   n26 = "Constraint - dyn_jacket_r_02 (P)"(ComponentPose=@n25, Alpha=1, ConstraintWeights_0=1) {
    alphaBoolBlend {blendOption=Linear}
    boneToModify {boneName=dyn_jacket_r_02}
    constraintSetup [{targetBone {boneName=spine_04}}]
    constraintWeights [1]
   }
   n27 = "Transform (Modify) Bone - Bone: dyn_jacket_bk_l_01_MSTR"(ComponentPose=@n26, Alpha=1, Translation="0,0,0", Rotation="10,0,0", Scale="1,1,1") {
    rotationMode=BMM_Additive
    rotationSpace=BCS_BoneSpace
    alphaBoolBlend {blendOption=Linear}
    alphaScaleBiasClamp {interpSpeedDecreasing=3}
    boneToModify {boneName=dyn_jacket_bk_l_01_MSTR}
    rotation {pitch=10}
   }
   n28 = "Transform (Modify) Bone - Bone: dyn_jacket_bk_01_MSTR"(ComponentPose=@n27, Alpha=1, Translation="0,0,0", Rotation="12,0,0", Scale="1,1,1") {
    rotationMode=BMM_Additive
    rotationSpace=BCS_BoneSpace
    alphaBoolBlend {blendOption=Linear}
    alphaScaleBiasClamp {interpSpeedDecreasing=3}
    boneToModify {boneName=dyn_jacket_bk_01_MSTR}
    rotation {pitch=12}
   }
   n29 = "Transform (Modify) Bone - Bone: dyn_jacket_bk_r_01_MSTR"(ComponentPose=@n28, Alpha=1, Translation="0,0,0", Rotation="15,0,0", Scale="1,1,1") {
    rotationMode=BMM_Additive
    rotationSpace=BCS_BoneSpace
    alphaBoolBlend {blendOption=Linear}
    alphaScaleBiasClamp {interpSpeedDecreasing=3}
    boneToModify {boneName=dyn_jacket_bk_r_01_MSTR}
    rotation {pitch=15}
   }
   n30 = "Transform (Modify) Bone - Bone: dyn_jacket_r_01_MSTR"(ComponentPose=@n29, Alpha=1, Translation="0,0,0", Rotation="12,0,0", Scale="1,1,1") {
    rotationMode=BMM_Additive
    rotationSpace=BCS_BoneSpace
    alphaBoolBlend {blendOption=Linear}
    alphaScaleBiasClamp {interpSpeedDecreasing=3}
    boneToModify {boneName=dyn_jacket_r_01_MSTR}
    rotation {pitch=12}
   }
   n31 = "Transform (Modify) Bone - Bone: dyn_jacket_l_01_MSTR"(ComponentPose=@n30, Alpha=1, Translation="0,0,0", Rotation="10,0,0", Scale="1,1,1") {
    rotationMode=BMM_Additive
    rotationSpace=BCS_BoneSpace
    alphaBoolBlend {blendOption=Linear}
    alphaScaleBiasClamp {interpSpeedDecreasing=3}
    boneToModify {boneName=dyn_jacket_l_01_MSTR}
    rotation {pitch=10}
   }
   Save cached pose 'preSimCache'(Pose=Component To Local(ComponentPose=@n31)) {cacheName=preSimCache}
   n32 = "Anim Dynamics - Chain: dyn_jacket_l_01 -> dyn_jacket_l_02"!(ComponentPose=Local To Component(LocalPose=Use cached pose 'preSimCache'()), ExternalForce="0,0,0") -> (Pose) {
    alpha=0.8
    angularDampingOverride=1
    bChain=true
    bOverrideAngularDamping=true
    bUseGravityOverride=true
    alphaBoolBlend {blendOption=Linear}
    boundBone {boneName=dyn_jacket_l_01}
    chainEnd {boneName=dyn_jacket_l_02}
    componentAppliedLinearAccClamp=<2000 2000 2000>
    componentLinearAccScale=<0.4 0.3 0.3>
    componentLinearVelScale=<0.4 0.3 0.3>
    gravityOverride {z=-980}
    physicsBodyDefinitions [
     {
      boundBone {boneName=dyn_jacket_l_01}
      boxExtents=<10 10 10>
      constraintSetup {angularLimitsMax=<1 0 1>, angularLimitsMin=<-1 -45 -1>, angularTarget=<0 0 0>}
      localJointOffset=<5 0 0>
     }
     {
      boundBone {boneName=dyn_jacket_l_02}
      boxExtents=<10 10 10>
      constraintSetup {angularLimitsMax=<1 0 5>, angularLimitsMin=<-1 -25 -5>, angularTarget=<0 0 0>}
      localJointOffset=<5 0 0>
     }
    ]
   }
   n33 = "Anim Dynamics - Chain: dyn_jacket_bk_l_01 -> dyn_jacket_bk_l_02"!(ComponentPose=@n32, ExternalForce="0,0,0") -> (Pose) {
    alpha=0.8
    angularDampingOverride=1
    bChain=true
    bOverrideAngularDamping=true
    bUseGravityOverride=true
    alphaBoolBlend {blendOption=Linear}
    boundBone {boneName=dyn_jacket_bk_l_01}
    chainEnd {boneName=dyn_jacket_bk_l_02}
    componentAppliedLinearAccClamp=<2000 2000 2000>
    componentLinearAccScale=<0.4 0.3 0.3>
    componentLinearVelScale=<0.4 0.3 0.3>
    gravityOverride {z=-980}
    physicsBodyDefinitions [
     {
      boundBone {boneName=dyn_jacket_bk_l_01}
      boxExtents=<10 10 10>
      constraintSetup {angularLimitsMax=<1 0 1>, angularLimitsMin=<-1 -45 -1>, angularTarget=<0 0 0>}
      localJointOffset=<5 0 0>
     }
     {
      boundBone {boneName=dyn_jacket_bk_l_02}
      boxExtents=<10 10 10>
      constraintSetup {angularLimitsMax=<1 0 5>, angularLimitsMin=<-1 -25 -5>, angularTarget=<0 0 0>}
      localJointOffset=<5 0 0>
     }
    ]
   }
   n34 = "Anim Dynamics - Chain: dyn_jacket_bk_01 -> dyn_jacket_bk_02"!(ComponentPose=@n33, ExternalForce="0,0,0") -> (Pose) {
    alpha=0.8
    angularDampingOverride=1
    bChain=true
    bOverrideAngularDamping=true
    bUseGravityOverride=true
    alphaBoolBlend {blendOption=Linear}
    boundBone {boneName=dyn_jacket_bk_01}
    chainEnd {boneName=dyn_jacket_bk_02}
    componentAppliedLinearAccClamp=<2000 2000 2000>
    componentLinearAccScale=<0.4 0.3 0.3>
    componentLinearVelScale=<0.4 0.3 0.3>
    gravityOverride {z=-980}
    physicsBodyDefinitions [
     {
      boundBone {boneName=dyn_jacket_bk_01}
      boxExtents=<10 10 10>
      constraintSetup {angularLimitsMax=<1 0 1>, angularLimitsMin=<-1 -45 -1>, angularTarget=<0 0 0>}
      localJointOffset=<5 0 0>
     }
     {
      boundBone {boneName=dyn_jacket_bk_02}
      boxExtents=<10 10 10>
      constraintSetup {angularLimitsMax=<1 0 5>, angularLimitsMin=<-1 -25 -5>, angularTarget=<0 0 0>}
      localJointOffset=<5 0 0>
     }
    ]
   }
   n35 = "Anim Dynamics - Chain: dyn_jacket_fr_l_01 -> dyn_jacket_fr_l_02"!(ComponentPose=@n34, ExternalForce="0,0,0") -> (Pose) {
    alpha=0.8
    angularDampingOverride=1
    bChain=true
    bOverrideAngularDamping=true
    bUseGravityOverride=true
    alphaBoolBlend {blendOption=Linear}
    boundBone {boneName=dyn_jacket_fr_l_01}
    chainEnd {boneName=dyn_jacket_fr_l_02}
    componentAppliedLinearAccClamp=<2000 2000 2000>
    componentLinearAccScale=<0.4 0.3 0.3>
    componentLinearVelScale=<0.4 0.3 0.3>
    gravityOverride {z=-980}
    physicsBodyDefinitions [
     {
      boundBone {boneName=dyn_jacket_fr_l_01}
      boxExtents=<10 10 10>
      constraintSetup {angularLimitsMax=<1 0 1>, angularLimitsMin=<-1 -45 -1>, angularTarget=<0 0 0>}
      localJointOffset=<5 0 0>
     }
     {
      boundBone {boneName=dyn_jacket_fr_l_02}
      boxExtents=<10 10 10>
      constraintSetup {angularLimitsMax=<1 0 5>, angularLimitsMin=<-1 -25 -5>, angularTarget=<0 0 0>}
      localJointOffset=<5 0 0>
     }
    ]
   }
   n36 = "Anim Dynamics - Chain: dyn_jacket_fr_01 -> dyn_jacket_fr_02"!(ComponentPose=@n35, ExternalForce="0,0,0") -> (Pose) {
    alpha=0.8
    angularDampingOverride=1
    bChain=true
    bOverrideAngularDamping=true
    bUseGravityOverride=true
    alphaBoolBlend {blendOption=Linear}
    boundBone {boneName=dyn_jacket_fr_01}
    chainEnd {boneName=dyn_jacket_fr_02}
    componentAppliedLinearAccClamp=<2000 2000 2000>
    componentLinearAccScale=<0.4 0.3 0.3>
    componentLinearVelScale=<0.4 0.3 0.3>
    gravityOverride {z=-980}
    physicsBodyDefinitions [
     {
      boundBone {boneName=dyn_jacket_fr_01}
      boxExtents=<10 10 10>
      constraintSetup {angularLimitsMax=<1 0 1>, angularLimitsMin=<-1 -45 -1>, angularTarget=<0 0 0>}
      localJointOffset=<5 0 0>
     }
     {
      boundBone {boneName=dyn_jacket_fr_02}
      boxExtents=<10 10 10>
      constraintSetup {angularLimitsMax=<1 0 5>, angularLimitsMin=<-1 -25 -5>, angularTarget=<0 0 0>}
      localJointOffset=<5 0 0>
     }
    ]
   }
   n37 = "Anim Dynamics - Chain: dyn_jacket_fr_r_01 -> dyn_jacket_fr_r_02"!(ComponentPose=@n36, ExternalForce="0,0,0") -> (Pose) {
    alpha=0.8
    angularDampingOverride=1
    bChain=true
    bOverrideAngularDamping=true
    bUseGravityOverride=true
    alphaBoolBlend {blendOption=Linear}
    boundBone {boneName=dyn_jacket_fr_r_01}
    chainEnd {boneName=dyn_jacket_fr_r_02}
    componentAppliedLinearAccClamp=<2000 2000 2000>
    componentLinearAccScale=<0.4 0.3 0.3>
    componentLinearVelScale=<0.4 0.3 0.3>
    gravityOverride {z=-980}
    physicsBodyDefinitions [
     {
      boundBone {boneName=dyn_jacket_fr_r_01}
      boxExtents=<10 10 10>
      constraintSetup {angularLimitsMax=<1 0 1>, angularLimitsMin=<-1 -45 -1>, angularTarget=<0 0 0>}
      localJointOffset=<5 0 0>
     }
     {
      boundBone {boneName=dyn_jacket_fr_r_02}
      boxExtents=<10 10 10>
      constraintSetup {angularLimitsMax=<1 0 5>, angularLimitsMin=<-1 -25 -5>, angularTarget=<0 0 0>}
      localJointOffset=<5 0 0>
     }
    ]
   }
   n38 = "Anim Dynamics - Chain: dyn_jacket_bk_r_01 -> dyn_jacket_bk_r_02"!(ComponentPose=@n37, ExternalForce="0,0,0") -> (Pose) {
    alpha=0.8
    angularDampingOverride=1
    bChain=true
    bOverrideAngularDamping=true
    bUseGravityOverride=true
    alphaBoolBlend {blendOption=Linear}
    boundBone {boneName=dyn_jacket_bk_r_01}
    chainEnd {boneName=dyn_jacket_bk_r_02}
    componentAppliedLinearAccClamp=<2000 2000 2000>
    componentLinearAccScale=<0.4 0.3 0.3>
    componentLinearVelScale=<0.4 0.3 0.3>
    gravityOverride {z=-980}
    physicsBodyDefinitions [
     {
      boundBone {boneName=dyn_jacket_bk_r_01}
      boxExtents=<10 10 10>
      constraintSetup {angularLimitsMax=<1 -5 1>, angularLimitsMin=<-1 -45 -1>, angularTarget=<0 0 0>}
      localJointOffset=<5 0 0>
     }
     {
      boundBone {boneName=dyn_jacket_bk_r_02}
      boxExtents=<10 10 10>
      constraintSetup {angularLimitsMax=<1 -5 5>, angularLimitsMin=<-1 -25 -5>, angularTarget=<0 0 0>}
      localJointOffset=<5 0 0>
     }
    ]
   }
   n39 = "Anim Dynamics - Chain: dyn_jacket_r_01 -> dyn_jacket_r_02"!(ComponentPose=@n38, ExternalForce="0,0,0") -> (Pose) {
    alpha=0.8
    angularDampingOverride=1
    bChain=true
    bOverrideAngularDamping=true
    bUseGravityOverride=true
    alphaBoolBlend {blendOption=Linear}
    boundBone {boneName=dyn_jacket_r_01}
    chainEnd {boneName=dyn_jacket_r_02}
    componentAppliedLinearAccClamp=<2000 2000 2000>
    componentLinearAccScale=<0.4 0.3 0.3>
    componentLinearVelScale=<0.4 0.3 0.3>
    gravityOverride {z=-980}
    physicsBodyDefinitions [
     {
      boundBone {boneName=dyn_jacket_r_01}
      boxExtents=<10 10 10>
      constraintSetup {angularLimitsMax=<1 0 1>, angularLimitsMin=<-1 -45 -1>, angularTarget=<0 0 0>}
      localJointOffset=<5 0 0>
     }
     {
      boundBone {boneName=dyn_jacket_r_02}
      boxExtents=<10 10 10>
      constraintSetup {angularLimitsMax=<1 0 5>, angularLimitsMin=<-1 -25 -5>, angularTarget=<0 0 0>}
      localJointOffset=<5 0 0>
     }
    ]
   }
   n40 = "Anim Dynamics - Chain: dyn_necklace_01 -> dyn_necklace_03"!(ComponentPose=@n39, ExternalForce="0,0,0") -> (Pose) {
    alpha=0.3
    angularDampingOverride=1
    bChain=true
    bOverrideAngularDamping=true
    bUseGravityOverride=true
    alphaBoolBlend {blendOption=Linear}
    boundBone {boneName=dyn_necklace_01}
    chainEnd {boneName=dyn_necklace_03}
    componentAppliedLinearAccClamp=<2000 2000 2000>
    componentLinearAccScale=<0.4 0.3 0.3>
    componentLinearVelScale=<0.4 0.3 0.3>
    gravityOverride {z=-980}
    physicsBodyDefinitions [
     {
      boundBone {boneName=dyn_necklace_01}
      boxExtents=<10 10 10>
      constraintSetup {angularLimitsMax=<1 0 30>, angularLimitsMin=<-1 -30 -30>, angularTarget=<0 0 0>}
      localJointOffset=<5 0 0>
     }
     {
      boundBone {boneName=dyn_necklace_02}
      boxExtents=<10 10 10>
      constraintSetup {angularLimitsMax=<1 0 30>, angularLimitsMin=<-1 -30 -30>, angularTarget=<0 0 0>}
      localJointOffset=<5 0 0>
     }
     {
      boundBone {boneName=dyn_necklace_03}
      boxExtents=<10 10 10>
      constraintSetup {angularLimitsMax=<1 0 30>, angularLimitsMin=<-1 -30 -30>, angularTarget=<0 0 0>}
      localJointOffset=<5 0 0>
     }
    ]
   }
   n41 = RigidBody!(ComponentPose=@n40, Alpha, OverrideWorldGravity:Vector="0,0,-980", ExternalForce) -> (Pose) {
    bClampLinearTranslationLimitToRefPose=true
    bOverrideWorldGravity=true
    overridePhysicsAsset=&/F_MED_Ocean_Rider_Dyn_Hips
    simulationSpace=BaseBoneSpace
    alphaBoolBlend {blendOption=Linear}
    baseBoneRef {boneName=pelvis}
    componentAppliedLinearAccClamp=<1000 1000 1000>
    componentLinearAccScale=<0.2 0.2 0.3>
    componentLinearVelScale=<0.2 0.2 0.3>
    overrideWorldGravity {z=-980}
   }
   n42 = RigidBody(ComponentPose=@n41, OverrideWorldGravity="0,0,-980") {
    bClampLinearTranslationLimitToRefPose=true
    bOverrideWorldGravity=true
    overridePhysicsAsset=&/F_MED_Ocean_Rider_Dyn_back_strap
    simulationSpace=BaseBoneSpace
    alphaBoolBlend {blendOption=Linear}
    baseBoneRef {boneName=pelvis}
    componentAppliedLinearAccClamp=<2000 2000 2000>
    componentLinearAccScale=<0.4 0.3 0.3>
    componentLinearVelScale=<0.4 0.3 0.3>
    overrideWorldGravity {z=-980}
   }
   n43 = RigidBody(ComponentPose=@n42, OverrideWorldGravity="0,0,-980") {
    bClampLinearTranslationLimitToRefPose=true
    bOverrideWorldGravity=true
    overridePhysicsAsset=&/F_MED_Ocean_Rider_Dyn_arm_l
    simulationSpace=BaseBoneSpace
    alphaBoolBlend {blendOption=Linear}
    baseBoneRef {boneName=upperarm_l}
    componentAppliedLinearAccClamp=<3000 3000 3000>
    componentLinearAccScale=<0.4 0.3 0.3>
    componentLinearVelScale=<0.4 0.3 0.3>
    overrideWorldGravity {z=-980}
   }
   n44 = RigidBody(ComponentPose=@n43, OverrideWorldGravity="0,0,-980") {
    bClampLinearTranslationLimitToRefPose=true
    bOverrideWorldGravity=true
    overridePhysicsAsset=&/F_MED_Ocean_Rider_Dyn_arm_r
    simulationSpace=BaseBoneSpace
    alphaBoolBlend {blendOption=Linear}
    baseBoneRef {boneName=upperarm_r}
    componentAppliedLinearAccClamp=<3000 3000 3000>
    componentLinearAccScale=<0.4 0.3 0.3>
    componentLinearVelScale=<0.4 0.3 0.3>
    overrideWorldGravity {z=-980}
   }
   Save cached pose 'crouchCache'(Pose=Component To Local(ComponentPose=@n44)) {cacheName=crouchCache}
  ]}
 EventGraph {nodes [
   Event BlueprintUpdateAnimation()
   > n1 = Sequence()
   > Set vestAlpha(vestAlpha:bool=OR Boolean!(A=Get bIsPlayingMeleeAnim(), B=Get bIsMoving2D(), C:bool=Get bIsJumping(), D:bool=Get bIsFalling()) -> (ReturnValue)) -> (Output_Get:bool)
   n2 = Gravity Override State Function DEPRECATED!(
    execute=@n1.then_1
    CharacterPartAnimInstance=Self-Reference()
    GameTimeInSeconds=GetGameTimeInSeconds()
    GravityOverrideParams_Skydive_GoingUp_28_#1_WindFrequency_28_#2:Vector="1.5,7,8"
    GravityOverrideParams_Skydive_GoingUp_28_#1_WindAmplitude_31_#3:Vector="200,-500,400"
    GravityOverrideParams_Skydive_GoingUp_28_#1_WindOffset_32_#4:Vector="0,0,0"
    GravityOverrideParams_Skydive_GoingUp_28_#1_GravityOverride_15_#5:Vector="0,0,980"
    GravityOverrideParams_Skydive_GoingUp_28_#1_JointName_27_#6:Name=root
    GravityOverrideParams_Skydive_FloatingDown_29_#7_WindFrequency_28_#2:Vector="1.5,7,8"
    GravityOverrideParams_Skydive_FloatingDown_29_#7_WindAmplitude_31_#3:Vector="200,-500,400"
    GravityOverrideParams_Skydive_FloatingDown_29_#7_WindOffset_32_#4:Vector="0,-600,400"
    GravityOverrideParams_Skydive_FloatingDown_29_#7_GravityOverride_15_#5:Vector="0,0,980"
    GravityOverrideParams_Skydive_FloatingDown_29_#7_JointName_27_#6:Name=root
    GravityOverrideParams_Skydive_DivingDown_30_#8_WindFrequency_28_#2:Vector="1.5,7,8"
    GravityOverrideParams_Skydive_DivingDown_30_#8_WindAmplitude_31_#3:Vector="50,300,100"
    GravityOverrideParams_Skydive_DivingDown_30_#8_WindOffset_32_#4:Vector="0,0,0"
    GravityOverrideParams_Skydive_DivingDown_30_#8_GravityOverride_15_#5:Vector="0,-980,0"
    GravityOverrideParams_Skydive_DivingDown_30_#8_JointName_27_#6:Name=pelvis
    GravityOverrideParams_Skydive_Parachuting_31_#9_WindFrequency_28_#2:Vector="1.5,7,8"
    GravityOverrideParams_Skydive_Parachuting_31_#9_WindAmplitude_31_#3:Vector="200,-500,400"
    GravityOverrideParams_Skydive_Parachuting_31_#9_WindOffset_32_#4:Vector="0,0,0"
    GravityOverrideParams_Skydive_Parachuting_31_#9_GravityOverride_15_#5:Vector="0,-800,-980"
    GravityOverrideParams_Skydive_Parachuting_31_#9_JointName_27_#6:Name=root
    GravityOverrideParams_OnGround_Standing_32_#10:Gravity Override Setting Struct Structure=(WindFrequency_28_#2=(X=0,Y=0,Z=0),WindAmplitude_31_#3=(X=0,Y=0,Z=0),WindOffset_32_#4=(X=0,Y=0,Z=0),GravityOverride_15_#5=(X=0,Y=0,Z=-980),JointName_27_#6="root")
    GravityOverrideParams_OnGround_Moving_33_#11:Gravity Override Setting Struct Structure=(WindFrequency_28_#2=(X=0,Y=0,Z=0),WindAmplitude_31_#3=(X=0,Y=0,Z=0),WindOffset_32_#4=(X=0,Y=0,Z=0),GravityOverride_15_#5=(X=0,Y=0,Z=-980),JointName_27_#6="root")
    GravityOverrideParams_EmoteOrMelee_34_#12:Gravity Override Setting Struct Structure=(WindFrequency_28_#2=(X=0,Y=0,Z=0),WindAmplitude_31_#3=(X=0,Y=0,Z=0),WindOffset_32_#4=(X=0,Y=0,Z=0),GravityOverride_15_#5=(X=0,Y=0,Z=-980),JointName_27_#6="root")
    GravityOverrideParams_Swimming_35_#13:Gravity Override Setting Struct Structure=(WindFrequency_28_#2=(X=0,Y=0,Z=0),WindAmplitude_31_#3=(X=0,Y=0,Z=0),WindOffset_32_#4=(X=0,Y=0,Z=0),GravityOverride_15_#5=(X=-800,Y=0,Z=0),JointName_27_#6="pelvis")
    GravityOverrideParams_Front_End_36_#14:Gravity Override Setting Struct Structure=(WindFrequency_28_#2=(X=0,Y=0,Z=0),WindAmplitude_31_#3=(X=0,Y=0,Z=0),WindOffset_32_#4=(X=0,Y=0,Z=0),GravityOverride_15_#5=(X=0,Y=0,Z=-980),JointName_27_#6="root")
    GravityOverrideParams_DBNO_40_#15:Gravity Override Setting Struct Structure=(WindFrequency_28_#2=(X=0,Y=0,Z=0),WindAmplitude_31_#3=(X=0,Y=0,Z=0),WindOffset_32_#4=(X=0,Y=0,Z=0),GravityOverride_15_#5=(X=-980,Y=400,Z=0),JointName_27_#6="pelvis")
    GravityOverrideParams_DBNO_Carried_41_#16:Gravity Override Setting Struct Structure=(WindFrequency_28_#2=(X=0,Y=0,Z=0),WindAmplitude_31_#3=(X=0,Y=0,Z=0),WindOffset_32_#4=(X=0,Y=0,Z=0),GravityOverride_15_#5=(X=-980,Y=500,Z=-500),JointName_27_#6="pelvis")
    GravityOverrideParams_OnGround_Falling_45_#17:Gravity Override Setting Struct Structure=(WindFrequency_28_#2=(X=0,Y=0,Z=0),WindAmplitude_31_#3=(X=0,Y=0,Z=0),WindOffset_32_#4=(X=0,Y=0,Z=0),GravityOverride_15_#5=(X=0,Y=0,Z=-980),JointName_27_#6="root")
    GravityOverrideParams_Ziplining_46_#18:Gravity Override Setting Struct Structure=(WindFrequency_28_#2=(X=0,Y=0,Z=0),WindAmplitude_31_#3=(X=0,Y=0,Z=0),WindOffset_32_#4=(X=0,Y=0,Z=0),GravityOverride_15_#5=(X=0,Y=0,Z=-980),JointName_27_#6="root")
    GravityOverrideParams_Skydive_Surfing_49_#19:Gravity Override Setting Struct Structure=(WindFrequency_28_#2=(X=1.5,Y=7,Z=8),WindAmplitude_31_#3=(X=200,Y=-500,Z=400),WindOffset_32_#4=(X=0,Y=0,Z=0),GravityOverride_15_#5=(X=500,Y=-300,Z=-980),JointName_27_#6="root")
    GravityOverrideParams_Riding_Standing_54_#20:Gravity Override Setting Struct Structure=(WindFrequency_28_#2=(X=0,Y=0,Z=0),WindAmplitude_31_#3=(X=0,Y=0,Z=0),WindOffset_32_#4=(X=0,Y=0,Z=0),GravityOverride_15_#5=(X=0,Y=-350,Z=-980),JointName_27_#6="root")
    GravityOverrideParams_Riding_Moving_55_#21:Gravity Override Setting Struct Structure=(WindFrequency_28_#2=(X=0,Y=0,Z=0),WindAmplitude_31_#3=(X=0,Y=0,Z=0),WindOffset_32_#4=(X=0,Y=0,Z=0),GravityOverride_15_#5=(X=0,Y=-450,Z=-450),JointName_27_#6="root")
    GravityOverrideParams_Riding_Moving_Targeting_58_#22:Gravity Override Setting Struct Structure=(WindFrequency_28_#2=(X=0,Y=0,Z=0),WindAmplitude_31_#3=(X=0,Y=0,Z=0),WindOffset_32_#4=(X=0,Y=0,Z=0),GravityOverride_15_#5=(X=0,Y=-200,Z=-980),JointName_27_#6="root")
    GravityOverrideParams_Tactical_Sprint_57_#23:Gravity Override Setting Struct Structure=(WindFrequency_28_#2=(X=0,Y=0,Z=0),WindAmplitude_31_#3=(X=0,Y=0,Z=0),WindOffset_32_#4=(X=0,Y=0,Z=0),GravityOverride_15_#5=(X=0,Y=-450,Z=-500),JointName_27_#6="root")
    GravityOverrideParams_Rail_Grinding_61_#24:Gravity Override Setting Struct Structure=(WindFrequency_28_#2=(X=0,Y=0,Z=0),WindAmplitude_31_#3=(X=0,Y=0,Z=0),WindOffset_32_#4=(X=0,Y=0,Z=0),GravityOverride_15_#5=(X=0,Y=0,Z=-980),JointName_27_#6="root")
    GravityOverrideParams_DBNO_Tumbling_67_#25:Gravity Override Setting Struct Structure=(WindFrequency_28_#2=(X=0,Y=0,Z=0),WindAmplitude_31_#3=(X=0,Y=0,Z=0),WindOffset_32_#4=(X=0,Y=0,Z=0),GravityOverride_15_#5=(X=-2000,Y=0,Z=0),JointName_27_#6="pelvis")
    GravityOverrideParams_DBNO_RollFwd_68_#26:Gravity Override Setting Struct Structure=(WindFrequency_28_#2=(X=0,Y=0,Z=0),WindAmplitude_31_#3=(X=0,Y=0,Z=0),WindOffset_32_#4=(X=0,Y=0,Z=0),GravityOverride_15_#5=(X=-2000,Y=0,Z=0),JointName_27_#6="pelvis")
    GravityOverrideParams_DBNO_RollBwd_69_#27:Gravity Override Setting Struct Structure=(WindFrequency_28_#2=(X=0,Y=0,Z=0),WindAmplitude_31_#3=(X=0,Y=0,Z=0),WindOffset_32_#4=(X=0,Y=0,Z=0),GravityOverride_15_#5=(X=-4000,Y=0,Z=0),JointName_27_#6="pelvis")
    GravityOverrideParams_DBNO_Impulse_70_#28:Gravity Override Setting Struct Structure=(WindFrequency_28_#2=(X=0,Y=0,Z=0),WindAmplitude_31_#3=(X=0,Y=0,Z=0),WindOffset_32_#4=(X=0,Y=0,Z=0),GravityOverride_15_#5=(X=-2000,Y=0,Z=0),JointName_27_#6="pelvis")
    RBAN Params
    DebugDraw
    DebugJointName
   ) -> (then, GravityOverride, CurrentStateOut, GravityOverrideSettingsOut, RBAN SettingsOut)
   > Set gravityOverride(gravityOverride:Vector=@n2) -> (Output_Get:Vector)
   > n3 = Gravity Override State Function DEPRECATED(
    CharacterPartAnimInstance=Self-Reference()
    GameTimeInSeconds=GetGameTimeInSeconds()
    GravityOverrideParams_Skydive_GoingUp_28_#1_WindFrequency_28_#2="0,0,0"
    GravityOverrideParams_Skydive_GoingUp_28_#1_WindAmplitude_31_#3="0,0,0"
    GravityOverrideParams_Skydive_GoingUp_28_#1_WindOffset_32_#4="0,0,0"
    GravityOverrideParams_Skydive_GoingUp_28_#1_GravityOverride_15_#5="0,0,500"
    GravityOverrideParams_Skydive_GoingUp_28_#1_JointName_27_#6=root
    GravityOverrideParams_Skydive_FloatingDown_29_#7_WindFrequency_28_#2="0,0,0"
    GravityOverrideParams_Skydive_FloatingDown_29_#7_WindAmplitude_31_#3="0,0,0"
    GravityOverrideParams_Skydive_FloatingDown_29_#7_WindOffset_32_#4="0,0,0"
    GravityOverrideParams_Skydive_FloatingDown_29_#7_GravityOverride_15_#5="0,0,500"
    GravityOverrideParams_Skydive_FloatingDown_29_#7_JointName_27_#6=root
    GravityOverrideParams_Skydive_DivingDown_30_#8_WindFrequency_28_#2="1.5,7,8"
    GravityOverrideParams_Skydive_DivingDown_30_#8_WindAmplitude_31_#3="50,300,100"
    GravityOverrideParams_Skydive_DivingDown_30_#8_WindOffset_32_#4="0,0,0"
    GravityOverrideParams_Skydive_DivingDown_30_#8_GravityOverride_15_#5="0,0,0"
    GravityOverrideParams_Skydive_DivingDown_30_#8_JointName_27_#6=pelvis
    GravityOverrideParams_Skydive_Parachuting_31_#9_WindFrequency_28_#2="0,0,0"
    GravityOverrideParams_Skydive_Parachuting_31_#9_WindAmplitude_31_#3="0,0,0"
    GravityOverrideParams_Skydive_Parachuting_31_#9_WindOffset_32_#4="0,0,0"
    GravityOverrideParams_Skydive_Parachuting_31_#9_GravityOverride_15_#5="0,0,-980"
    GravityOverrideParams_Skydive_Parachuting_31_#9_JointName_27_#6=root
    GravityOverrideParams_OnGround_Standing_32_#10=(WindFrequency_28_#2=(X=0,Y=0,Z=0),WindAmplitude_31_#3=(X=0,Y=0,Z=0),WindOffset_32_#4=(X=0,Y=0,Z=0),GravityOverride_15_#5=(X=0,Y=0,Z=-980),JointName_27_#6="root")
    GravityOverrideParams_OnGround_Moving_33_#11=(WindFrequency_28_#2=(X=0,Y=0,Z=0),WindAmplitude_31_#3=(X=0,Y=0,Z=0),WindOffset_32_#4=(X=0,Y=0,Z=0),GravityOverride_15_#5=(X=0,Y=0,Z=-980),JointName_27_#6="root")
    GravityOverrideParams_EmoteOrMelee_34_#12=(WindFrequency_28_#2=(X=0,Y=0,Z=0),WindAmplitude_31_#3=(X=0,Y=0,Z=0),WindOffset_32_#4=(X=0,Y=0,Z=0),GravityOverride_15_#5=(X=0,Y=0,Z=-980),JointName_27_#6="root")
    GravityOverrideParams_Swimming_35_#13=(WindFrequency_28_#2=(X=0,Y=0,Z=0),WindAmplitude_31_#3=(X=0,Y=0,Z=0),WindOffset_32_#4=(X=0,Y=0,Z=0),GravityOverride_15_#5=(X=-800,Y=0,Z=0),JointName_27_#6="pelvis")
    GravityOverrideParams_Front_End_36_#14=(WindFrequency_28_#2=(X=0,Y=0,Z=0),WindAmplitude_31_#3=(X=0,Y=0,Z=0),WindOffset_32_#4=(X=0,Y=0,Z=0),GravityOverride_15_#5=(X=0,Y=0,Z=-980),JointName_27_#6="root")
    GravityOverrideParams_DBNO_40_#15=(WindFrequency_28_#2=(X=0,Y=0,Z=0),WindAmplitude_31_#3=(X=0,Y=0,Z=0),WindOffset_32_#4=(X=0,Y=0,Z=0),GravityOverride_15_#5=(X=-980,Y=400,Z=0),JointName_27_#6="pelvis")
    GravityOverrideParams_DBNO_Carried_41_#16=(WindFrequency_28_#2=(X=0,Y=0,Z=0),WindAmplitude_31_#3=(X=0,Y=0,Z=0),WindOffset_32_#4=(X=0,Y=0,Z=0),GravityOverride_15_#5=(X=-980,Y=500,Z=-500),JointName_27_#6="pelvis")
    GravityOverrideParams_OnGround_Falling_45_#17=(WindFrequency_28_#2=(X=0,Y=0,Z=0),WindAmplitude_31_#3=(X=0,Y=0,Z=0),WindOffset_32_#4=(X=0,Y=0,Z=0),GravityOverride_15_#5=(X=0,Y=0,Z=-980),JointName_27_#6="root")
    GravityOverrideParams_Ziplining_46_#18=(WindFrequency_28_#2=(X=0,Y=0,Z=0),WindAmplitude_31_#3=(X=0,Y=0,Z=0),WindOffset_32_#4=(X=0,Y=0,Z=0),GravityOverride_15_#5=(X=0,Y=0,Z=-980),JointName_27_#6="root")
    GravityOverrideParams_Skydive_Surfing_49_#19=(WindFrequency_28_#2=(X=1.5,Y=7,Z=8),WindAmplitude_31_#3=(X=200,Y=-500,Z=400),WindOffset_32_#4=(X=0,Y=0,Z=0),GravityOverride_15_#5=(X=500,Y=-300,Z=-980),JointName_27_#6="root")
    GravityOverrideParams_Riding_Standing_54_#20=(WindFrequency_28_#2=(X=0,Y=0,Z=0),WindAmplitude_31_#3=(X=0,Y=0,Z=0),WindOffset_32_#4=(X=0,Y=0,Z=0),GravityOverride_15_#5=(X=0,Y=-350,Z=-980),JointName_27_#6="root")
    GravityOverrideParams_Riding_Moving_55_#21=(WindFrequency_28_#2=(X=0,Y=0,Z=0),WindAmplitude_31_#3=(X=0,Y=0,Z=0),WindOffset_32_#4=(X=0,Y=0,Z=0),GravityOverride_15_#5=(X=0,Y=-450,Z=-450),JointName_27_#6="root")
    GravityOverrideParams_Riding_Moving_Targeting_58_#22=(WindFrequency_28_#2=(X=0,Y=0,Z=0),WindAmplitude_31_#3=(X=0,Y=0,Z=0),WindOffset_32_#4=(X=0,Y=0,Z=0),GravityOverride_15_#5=(X=0,Y=-200,Z=-980),JointName_27_#6="root")
    GravityOverrideParams_Tactical_Sprint_57_#23=(WindFrequency_28_#2=(X=0,Y=0,Z=0),WindAmplitude_31_#3=(X=0,Y=0,Z=0),WindOffset_32_#4=(X=0,Y=0,Z=0),GravityOverride_15_#5=(X=0,Y=-450,Z=-500),JointName_27_#6="root")
    GravityOverrideParams_Rail_Grinding_61_#24=(WindFrequency_28_#2=(X=0,Y=0,Z=0),WindAmplitude_31_#3=(X=0,Y=0,Z=0),WindOffset_32_#4=(X=0,Y=0,Z=0),GravityOverride_15_#5=(X=0,Y=0,Z=-980),JointName_27_#6="root")
    GravityOverrideParams_DBNO_Tumbling_67_#25=(WindFrequency_28_#2=(X=0,Y=0,Z=0),WindAmplitude_31_#3=(X=0,Y=0,Z=0),WindOffset_32_#4=(X=0,Y=0,Z=0),GravityOverride_15_#5=(X=-2000,Y=0,Z=0),JointName_27_#6="pelvis")
    GravityOverrideParams_DBNO_RollFwd_68_#26=(WindFrequency_28_#2=(X=0,Y=0,Z=0),WindAmplitude_31_#3=(X=0,Y=0,Z=0),WindOffset_32_#4=(X=0,Y=0,Z=0),GravityOverride_15_#5=(X=-2000,Y=0,Z=0),JointName_27_#6="pelvis")
    GravityOverrideParams_DBNO_RollBwd_69_#27=(WindFrequency_28_#2=(X=0,Y=0,Z=0),WindAmplitude_31_#3=(X=0,Y=0,Z=0),WindOffset_32_#4=(X=0,Y=0,Z=0),GravityOverride_15_#5=(X=-4000,Y=0,Z=0),JointName_27_#6="pelvis")
    GravityOverrideParams_DBNO_Impulse_70_#28=(WindFrequency_28_#2=(X=0,Y=0,Z=0),WindAmplitude_31_#3=(X=0,Y=0,Z=0),WindOffset_32_#4=(X=0,Y=0,Z=0),GravityOverride_15_#5=(X=-2000,Y=0,Z=0),JointName_27_#6="pelvis")
   )
   > Set gravityForLegsPhat(gravityForLegsPhat:Vector=@n3) -> (Output_Get:Vector)
   Event BlueprintInitializeAnimation()
   > Set Skel_Mesh(Skel_Mesh=GetOwnerSkeletalMeshComponent())
  ]}
]
variables [
 gravityForLegsPhat {type=Vector}
 gravityOverride {type=Vector}
 Skel_Mesh {type=SkeletalMeshComponent}
 vestAlpha {type=boolean}
]
