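# Physics asset for the Ocean Rider female medium helmet part, defining dynamics for the front hair strands and ponytail.
# Capsule bodies cover the head, left and right front hair tips and four ponytail bones.
# Both front strands and the ponytail root constrain to the head with tight twist and wider swing limits.
# The other three ponytail bodies all constrain directly to the root rather than forming a sequential chain, with narrower swing.
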PhysicsAsset /BRCosmetics/Characters/Player/Female/Medium/Bodies/F_MED_Ocean_Rider/Meshes/Parts/F_MED_Ocean_Rider_Helmet_Dyn
bodies {
 dyn_hair_fr_l_03 [{shapeType=Capsule, length=4.1475, radius=1.1737, center=<3.1313 0 0>, rotation=<rot 90 0 0>}]
 dyn_hair_fr_r_03 [{shapeType=Capsule, length=4.1475, radius=1.1737, center=<3.1313 0 0>, rotation=<rot 90 0 0>}]
 dyn_hair_ponytail_01 [{shapeType=Capsule, length=3.6787, radius=1.6424, center=<3.1313 0 0>, rotation=<rot 90 0 0>}]
 dyn_hair_ponytail_02 [{shapeType=Capsule, length=3.6787, radius=1.6424, center=<3.1313 0 0>, rotation=<rot 90 0 0>}]
 dyn_hair_ponytail_03 [{shapeType=Capsule, length=3.6787, radius=1.6424, center=<3.1313 0 0>, rotation=<rot 90 0 0>}]
 dyn_hair_ponytail_04 [{shapeType=Capsule, length=3.6787, radius=1.6424, center=<3.1313 0 0>, rotation=<rot 90 0 0>}]
 head [{shapeType=Capsule, length=9.368, radius=6.2949, rotation=<rot 90 0 0>}]
}
constraints [
 {bone1Name=dyn_hair_fr_l_03, bone2Name=head, swing1LimitDegrees=45, swing2LimitDegrees=45, twistLimitDegrees=5}
 {bone1Name=dyn_hair_fr_r_03, bone2Name=head, swing1LimitDegrees=45, swing2LimitDegrees=45, twistLimitDegrees=5}
 {bone1Name=dyn_hair_ponytail_01, bone2Name=head, swing1LimitDegrees=60, swing2LimitDegrees=60, twistLimitDegrees=5}
 {bone1Name=dyn_hair_ponytail_02, bone2Name=dyn_hair_ponytail_01, swing1LimitDegrees=30, swing2LimitDegrees=30, twistLimitDegrees=5}
 {bone1Name=dyn_hair_ponytail_03, bone2Name=dyn_hair_ponytail_01, swing1LimitDegrees=30, swing2LimitDegrees=30, twistLimitDegrees=5}
 {bone1Name=dyn_hair_ponytail_04, bone2Name=dyn_hair_ponytail_01, swing1LimitDegrees=30, swing2LimitDegrees=30, twistLimitDegrees=5}
]
